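# Conversation database for the NPC Goldie, defining the heist-themed greeting given when a player starts talking to this character.
# The Goldie entry point leads to one Speech node with a default line and prioritized contextual alternatives, including a first-ever-conversation line.
# The other alternatives have empty requirement lists and name an OutfitReaction priority row with no curve table.
# The speech branches through link nodes into the shared NPCLibrary Quests and BasicServices conversations.
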ConversationDatabase /NPCLibrary/NPCs/Goldie/NPCLibrary_Conversation_Goldie
entry_points [{tag=Conversation.Entry.Goldie, destinations [#1]}]
nodes [
 ConversationEntryPointNode_7 {
  id=#1
  class=ConversationEntryPointNode
  outputs [#2]
  properties {entryTag {tagName=Conversation.Entry.Goldie}}
 }
 Speech_C_4 {
  id=#2
  class=Speech_C
  outputs [#3, #4]
  properties {
   nodeName=""
   generalConfig {
    defaultMessage=I'm ready to hit the Cold Blooded when you are.
    contextualMessages [
     {
      message=Renegade Shadow told me you might be in for a heist.
      contextRequirements [{requirement {class=FortControllerRequirement_IsFirstEverConversationWithNPC}}]
      priority {curve {rowName=Tandem.Speech.Priority.FirstEverInteraction}}
     }
     {
      message=I'm ready to hit the Cold Blooded when you are.
      requirementMatchPolicy=RequireAny
      contextRequirements []
      priority {curve {curveTable=null, rowName=Tandem.Speech.Priority.OutfitReaction}}
     }
     {
      message=I keep it a little dangerous and a little glamorous.
      requirementMatchPolicy=RequireAny
      contextRequirements []
      priority {curve {curveTable=null, rowName=Tandem.Speech.Priority.OutfitReaction}}
     }
     {
      message=Do you ever wish you could turn things gold with just a touch?
      requirementMatchPolicy=RequireAny
      contextRequirements []
      priority {curve {curveTable=null, rowName=Tandem.Speech.Priority.OutfitReaction}}
     }
     {
      message=Rule number one: don't get caught.
      requirementMatchPolicy=RequireAny
      contextRequirements []
      priority {curve {curveTable=null, rowName=Tandem.Speech.Priority.OutfitReaction}}
     }
    ]
   }
  }
 }
 ConversationLinkNode_0 {id=#4, class=ConversationLinkNode, properties {remoteEntryTag {tagName=Conversation.Entry.BasicServices}}}
 ConversationLinkNode_1 {id=#3, class=ConversationLinkNode, properties {remoteEntryTag {tagName=Conversation.Entry.NPCLibrary.Quests}}}
]
